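# Niagara system for the FlourCut emote, named as its prop spawn effect.
# The Sparkles emitter fires a one-shot burst of warm gold sprites from a flat box, which rise with upward gravity, curl noise and drag while fading out.
# Sparkle, airglow and smoke LWE emitters show no modules or renderers here.
# A disabled MusicNotes emitter holds velocity-dependent spawning, palette colors and time-of-day tinting.
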NiagaraSystem /CosmeticCompanions/FX/Builtins/FlourCut/NS_Emote_FlourCut_PropSpawn
dependencies {
 usedDataInterfaces [
  &NiagaraDataInterfaceCurve
  &NiagaraDataInterfaceArrayColor
  &NiagaraDataInterfaceSpriteRendererInfo
  &NiagaraDataInterfaceVectorCurve
 ]
 usedDynamicInputs [&MaskVector]
 usedModules [&SystemState]
}
emitters {
 AirglowLWE001 {rendererClasses [], renderers []}
 MusicNotes {
  bEnabled=false
  emitterUpdateScript [
   &EmitterState()
   &SpawnRate(
    SpawnRate=&MakeCustomFloatFromBool(
     Bool=&SetBoolByFloatComparison(A=&VectorLength(Vector=$User.OwnerLocalVelocity), B=310)
     True Float=&Multiply_Float(A=$User.SpawnCurve, B=3)
     False Float=&Multiply_Float(A=$User.SpawnCurve, B=7)
    )
    Use Spawn Probability=true
   )
  ]
  particleSpawnScript [
   &PartitionParticles(
    Particles Per Partition=1
    Partition Count Mode:ENiagara_InfiniteLoopDuration=::NewEnumerator1
    Partition Mode:ENiagara_PartitionMode_v2=::NewEnumerator1
   )
   &InitializeParticle(
    Color=&SelectLinearColorFromArray(
     Array Sampling Mode=::NewEnumerator1
     Direct Array Index=$Particles.PartitionParticles.Partition
     Direct Array Mode=::NewEnumerator1
     Color Selection Array=DataInterface() {propertyValues {ColorData [
        <rgba 0.8802 0.4371 0 1>
        <rgba 0.1009 0.4463 1 1>
        <rgba 1 0.0447 0 1>
        <rgba 0.3356 0.776 0.0243 1>
        <rgba 1 0.125 0.6934 1>
        <rgba 0.2561 0.1157 0.8229 1>
       ]}}
     Random Seed=?
    )
    Lifetime=&MakeCustomFloatFromBool(
     Bool=&SetBoolByFloatComparison(A=&VectorLength(Vector=$Engine.Owner.Velocity), B=450)
     True Float=0.8
     False Float=&RandomRangeFloat(Minimum=1.9, Maximum=2.35)
    )
    Sprite Rotation Angle=&RandomRangeFloat(Minimum=30, Maximum=-30)
    Sprite Rotation Mode=::NewEnumerator0
    Sprite Size Mode=::NewEnumerator1
    Uniform Sprite Size Max=8
    Uniform Sprite Size Min=4.5
    UsePositionOffset=false
    Write Material Random=true
   )
   &ShapeLocation(
    Handle Radius=0
    Large Radius=8
    MaskSpawnGroup=false
    Non Uniform Scale=<1 1 0>
    Offset Mode=::NewEnumerator0
    Override Seed=false
    Shape Primitive=::NewEnumerator3
    UseOwnerScale=true
   )
   &ApplyInitialForces()
   &AddVelocity(Velocity=&MakeCustomVectorFromBool(
     Bool=$User.bIsFrontEndPreview
     True Vector=&RandomRangeVector(Minimum=<0 0 155>, Maximum=<0 0 215>)
     False Vector=&RandomRangeVector(Minimum=<0 0 155>, Maximum=<0 0 215>)
    )) {module=AddVelocity001}
   &SubUVAnimation(
    Direct Frame Index=4
    "Sprite Renderer "=DataInterface() {propertyValues {SpriteRenderer=&self:MusicNotes_0.NiagaraSpriteRendererProperties_1}}
    SubUV Animation Mode:ENiagara_SubUVLookupModeV2=::NewEnumerator5
   )
   &SubUVAnimation(
    Direct Frame Index=5
    "Sprite Renderer "=DataInterface() {propertyValues {SpriteRenderer=&self:MusicNotes_0.NiagaraSpriteRendererProperties_0}}
    SubUV Animation Mode:ENiagara_SubUVLookupModeV2=::NewEnumerator5
   ) {module=SubUVAnimation001}
  ]
  particleUpdateScript [
   &ParticleState()
   &CurlNoiseForce(
    Noise Strength=&FloatFromCurve(
     FloatCurve=DataInterface() {propertyValues {Curve {keys [
         {interpMode=RCIM_Cubic, leaveTangentWeight=53.6425, tangentMode=RCTM_User, time=0, value=50}
         {arriveTangent=342.4859, interpMode=RCIM_Cubic, leaveTangent=342.4859, tangentMode=RCTM_User, time=1, value=150}
        ]}}}
     Scale Curve=&RandomRangeFloat(Minimum=5, Maximum=1)
    )
    Pan Noise=false
   )
   &GravityForce(Gravity=&RandomRangeVector(
     Minimum=&VectorFromCurve(VectorCurve=DataInterface() {propertyValues {
        XCurve {keys [<at 0 0>,<at 1 0>]}
        YCurve {keys [<at 0 0>,<at 1 0>]}
        ZCurve {keys [
          {interpMode=RCIM_Cubic, tangentMode=RCTM_User, time=0, value=10}
          {arriveTangent=88.4209, interpMode=RCIM_Cubic, leaveTangent=88.4209, tangentMode=RCTM_User, time=0.75, value=35}
         ]}
       }})
     Maximum=&VectorFromCurve(VectorCurve=DataInterface() {propertyValues {
        XCurve {keys [<at 0 0>,<at 1 0>]}
        YCurve {keys [<at 0 0>,<at 1 0>]}
        ZCurve {keys [
          {interpMode=RCIM_Cubic, tangentMode=RCTM_User, time=0, value=10}
          {arriveTangent=133.5137, interpMode=RCIM_Cubic, leaveTangent=133.5137, tangentMode=RCTM_User, time=0.75, value=70}
         ]}
       }})
    ))
   &DynamicMaterialParameters(
    Index 0 Param 1=&FloatFromCurve(FloatCurve=DataInterface() {propertyValues {Curve {keys [
         {arriveTangent=0.1706, interpMode=RCIM_Cubic, leaveTangent=0.1706, leaveTangentWeight=0.3081, tangentMode=RCTM_User, time=0, value=0}
         {arriveTangentWeight=0.3104, interpMode=RCIM_Cubic, leaveTangentWeight=0.0345, tangentMode=RCTM_User, time=0.9, value=0.2}
         {arriveTangent=3.2005, interpMode=RCIM_Cubic, leaveTangent=3.2005, tangentMode=RCTM_User, time=1, value=0.5}
        ]}}})
    Index 0 Param 2=0.15
    Module.Index0Float0Write=false
    Module.Index0Float1Write=true
    Module.Index0Float2Write=true
    Module.Index0Float3Write=true
   )
   &ScaleColor(
    Scale Alpha=&FloatFromCurve(
     FloatCurve=DataInterface() {propertyValues {Curve {
        defaultValue=0.3
        keys [
         {tangentMode=RCTM_User, time=0, value=1}
         {arriveTangent=-0.0405, interpMode=RCIM_Cubic, leaveTangent=-0.0405, tangentMode=RCTM_User, time=0.7009, value=1}
         {arriveTangent=-4.9566, interpMode=RCIM_Cubic, leaveTangent=-4.9566, tangentMode=RCTM_User, time=1, value=0.5}
        ]
       }}}
     Scale Curve=&FNTimeOfDayScalar(Night Scale=1, Morning Scale=1, Evening Scale=1)
    )
    Scale RGB=&MakeVectorFromLinearColorRGB(Color=&FNTimeOfDayColor(
      Night Color=<rgba 0.3609 0.4715 0.5469 1>
      Morning Color=<rgba 0.651 0.4517 0.4205 1>
      Day Color=<rgba 1 1 1 1>
      Evening Color=<rgba 0.651 0.4517 0.4205 1>
     ))
    ScaleA=true
    ScaleRGB=true
   )
   &ParticleState() {module=ParticleState001}
   &ScaleSpriteSize(Scale Factor=&Vector2DFromFloat(Value=&FloatFromCurve(
      FloatCurve=DataInterface() {propertyValues {Curve {keys [
          {arriveTangent=4.1785, interpMode=RCIM_Cubic, leaveTangent=4.1785, tangentMode=RCTM_User, time=0, value=0}
          {arriveTangent=3.6163, interpMode=RCIM_Cubic, leaveTangent=3.6163, tangentMode=RCTM_User, time=0.0913, value=0.5006}
          {
           arriveTangent=0.6479
           arriveTangentWeight=0.0521
           interpMode=RCIM_Cubic
           leaveTangent=0.6479
           leaveTangentWeight=0.2507
           tangentMode=RCTM_User
           time=0.2305
           value=0.7396
          }
          {arriveTangentWeight=0.2406, interpMode=RCIM_Cubic, leaveTangentWeight=0.0094, tangentMode=RCTM_User, time=0.9719, value=1}
          <at 1 2 RCIM_Cubic>
         ]}}}
      Scale Curve=4
     )))
   &Drag(
    Drag=&MakeFloatFromVector(
     Vector=&MakeCustomVectorFromBool(
      Bool=$User.bIsFrontEndPreview
      True Vector=&VectorFromFloat(Value=&RandomRangeFloat(Minimum=1, Maximum=4.5))
      False Vector=&VectorFromFloat(Value=&RandomRangeFloat(Minimum=1, Maximum=4.5))
     )
     Channel=::NewEnumerator0
    )
    Use Linear Drag=true
    Use Rotational Drag=true
   )
   &SolveForcesAndVelocity(Clamp Velocity=false, Limit Acceleration=false)
  ]
  rendererClasses [&NiagaraSpriteRendererProperties]
  renderers [
   {rendererClass=&NiagaraSpriteRendererProperties}
   {rendererClass=&NiagaraSpriteRendererProperties, rendererIndex=1}
  ]
 }
 SmokeLWE {rendererClasses [], renderers []}
 Sparkles {
  emitterUpdateScript [
   &EmitterState()
   &SpawnBurst_Instantaneous(Spawn Count=10, Spawn Time=0.06, Use Spawn Probability=false)
  ]
  particleSpawnScript [
   /SetVariables(Particles.VisibilityTag=&SpawnGroup()) {isSetParameters=true}
   &PartitionParticles(
    Number Of Partitions=&MergeHelper_IntSwitch(0=1, 1=1, 2=?, 3=?, 4=?, 5=?, Index=&SpawnGroup(), Num=::NewEnumerator0)
    Partition Mode:ENiagara_PartitionMode_v2=::NewEnumerator1
   ) {enabled=false}
   &InitializeParticle(
    Color=<rgba 2 1.0174 0.2259 1>
    Lifetime=&UniformRangedFloat(Minimum=0.3, Maximum=0.55)
    Sprite Rotation Angle=&RandomRangeFloat(Minimum=-90, Maximum=90)
    Sprite Rotation Mode=::NewEnumerator0
    Sprite Size Mode=::NewEnumerator3
    Uniform Sprite Size=&UniformRangedFloat(Minimum=5, Maximum=30)
    UsePositionOffset=false
    Write Material Random=true
   ) {module=InitializeParticle001}
   &AddVelocity(Velocity=&UniformRangedVector(Minimum=<-50 0 10>, Maximum=<50 0 150>))
   &ApplyInitialForces()
   &ShapeLocation(
    Box Midpoint=<0.5 0.5 1>
    Box Size=<120 25 0>
    MaskSpawnGroup=false
    Offset=<0 0 7>
    Offset Mode=::NewEnumerator0
    Override Seed=false
    Shape Primitive=::NewEnumerator2
    Surface Only=false
    UseOwnerScale=true
   ) {module=ShapeLocation002}
   &SubUVAnimation(End Frame=3, Start Frame=1, SubUV Animation Mode:ENiagara_SubUVLookupMode=::NewEnumerator2)
  ]
  particleUpdateScript [
   &ParticleState()
   &ScaleSpriteSize(Scale Factor=&Vector2DFromFloat(Value=&FloatFromCurve(
      FloatCurve=DataInterface() {propertyValues {Curve {
         defaultValue=0
         keys [
          {interpMode=RCIM_Cubic, tangentMode=RCTM_Break, time=0, value=0}
          {interpMode=RCIM_Cubic, leaveTangent=-0.1735, tangentMode=RCTM_Break, time=0.15, value=1.75}
          {arriveTangent=-0.177, interpMode=RCIM_Cubic, leaveTangent=-0.177, tangentMode=RCTM_User, time=0.3, value=1}
          {arriveTangent=-0.15, interpMode=RCIM_Cubic, leaveTangent=-0.15, tangentMode=RCTM_User, time=0.9574, value=1}
         ]
        }}}
      Scale Curve=0.7
     )))
   &SpriteRotationRate(Rotation Rate=&FloatFromCurve(FloatCurve=DataInterface() {propertyValues {Curve {keys [<at 0 1>,<at 0.3791 0>]}}}))
   &GravityForce(Gravity=<0 0 200>)
   /Niagara/Modules/Update/Forces/CurlNoiseForce(
    Noise Frequency=0.02
    Noise Quality / Cost=::NewEnumerator0
    Noise Strength=&MergeHelper_FloatSwitch(0=500, 1=200, 2=?, 3=?, Index=&SpawnGroup(), Num=::NewEnumerator0)
    Pan Noise=false
   )
   &Drag(Drag=&RandomRangeFloat(Minimum=2.5, Maximum=5), Use Linear Drag=true, Use Rotational Drag=false)
   &SolveForcesAndVelocity(Clamp Velocity=false, Limit Acceleration=false)
   &ScaleColor(
    Scale Alpha=&FloatFromCurve(FloatCurve=DataInterface() {propertyValues {Curve {keys [
         {interpMode=RCIM_Cubic, tangentMode=RCTM_Break, time=0, value=0.0144}
         {interpMode=RCIM_Cubic, tangentMode=RCTM_Break, time=0.2, value=0.9767}
         {interpMode=RCIM_Cubic, tangentMode=RCTM_Break, time=0.7949, value=-0.0164}
        ]}}})
    ScaleA=true
    ScaleRGB=true
   )
  ]
  rendererClasses [&NiagaraSpriteRendererProperties]
  renderers [{rendererClass=&NiagaraSpriteRendererProperties}]
 }
 SparklesLWE001 {rendererClasses [], renderers []}
}
summary {
 systemName=NS_Emote_FlourCut_PropSpawn
 emitters [
  {emitterName=SparklesLWE001, rendererClasses []}
  {emitterName=AirglowLWE001, rendererClasses []}
  {emitterName=SmokeLWE, rendererClasses []}
  {emitterName=Sparkles, rendererClasses [&NiagaraSpriteRendererProperties]}
  {bEnabled=false, emitterName=MusicNotes, rendererClasses [&NiagaraSpriteRendererProperties]}
 ]
 userVariables [
  User.bIsFrontEndPreview {type=NiagaraBool}
  User.OwnerLocalVelocity {type=Vector3f, description="", defaultValue {struct=Vector3f, value=<0 0 0>}}
  User.SpawnCurve {type=NiagaraFloat, defaultValue=1, description=""}
 ]
}
